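# Sound cue for the Cuddle Team Dark pickaxe impact, apparently for hits on enemies.
# A fixed short delay leads into a mixer that layers one of six pickaxe-specific enemy impacts with one of four shared flesh impacts.
# A modulator lowers the flesh layer to roughly half volume with slight variation.
# Shared pickaxe attenuation and melee concurrency settings are referenced, and the cue primes on load.
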SoundCue /BRCosmetics/Sounds/PickAxes/Cuddle_Team_Dark/PA_CuddleTeamDark_Impact_Cue
root=Delay_1
nodes [
 {id=Delay_1, children [Mixer_0], properties {delayMax=0.1, delayMin=0.1}}
 {id=Mixer_0, children [Random_2,Modulator_1], properties {inputVolume [1,1]}}
 {id=Modulator_1, children [Random_0], properties {volumeMax=0.5, volumeMin=0.4}}
 {id=Random_0, children [WavePlayer_0,WavePlayer_1,WavePlayer_2,WavePlayer_3], properties {weights [1,1,1,1]}}
 {
  id=Random_2
  children [WavePlayer_8, WavePlayer_10, WavePlayer_11, WavePlayer_12, WavePlayer_13, WavePlayer_14]
  properties {weights [1,1,1,1,1,1]}
 }
 {id=WavePlayer_0, properties {soundWaveAssetPtr=/Game/Sounds/Fort_Impact_Sounds/Flesh/new_player_impact_nonspat_A_01n}}
 {id=WavePlayer_1, properties {soundWaveAssetPtr=/Game/Sounds/Fort_Impact_Sounds/Flesh/new_player_impact_nonspat_A_02n}}
 {id=WavePlayer_10, properties {soundWaveAssetPtr=&/PickAxe_CuddleTeamDark_EnemyImpact_01}}
 {id=WavePlayer_11, properties {soundWaveAssetPtr=&/PickAxe_CuddleTeamDark_EnemyImpact_02}}
 {id=WavePlayer_12, properties {soundWaveAssetPtr=&/PickAxe_CuddleTeamDark_EnemyImpact_03}}
 {id=WavePlayer_13, properties {soundWaveAssetPtr=&/PickAxe_CuddleTeamDark_EnemyImpact_04}}
 {id=WavePlayer_14, properties {soundWaveAssetPtr=&/PickAxe_CuddleTeamDark_EnemyImpact_05}}
 {id=WavePlayer_2, properties {soundWaveAssetPtr=/Game/Sounds/Fort_Impact_Sounds/Flesh/new_player_impact_nonspat_A_03n}}
 {id=WavePlayer_3, properties {soundWaveAssetPtr=/Game/Sounds/Fort_Impact_Sounds/Flesh/new_player_impact_nonspat_A_04n}}
 {id=WavePlayer_8, properties {soundWaveAssetPtr=&/PickAxe_CuddleTeamDark_EnemyImpact_06}}
]
properties {
 attenuationSettings=/Game/Sounds/Attenuation/Fort_Pickaxe_Impact_Att
 audioPropertiesSheet=/BRCosmetics/Sounds/Settings/PropertySheets/Pickaxe/APS_Cosmetic_Pickaxe_Impact
 bPrimeOnLoad=true
 maxDistance=4300
 soundClassObject=&Fort_Impacts
 attenuationOverrides {
  bAttenuateWithLPF=true
  distanceAlgorithm=Logarithmic
  falloffDistance=800
  lPFRadiusMax=800
  lPFRadiusMin=400
  attenuationShapeExtents {x=300}
 }
 concurrencyOverrides {bLimitToOwner=true, maxCount=2, resolutionRule=StopOldest}
 concurrencySet [&Player_Melee_Weapon_Conc]
}
