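# Sound cue for the Cuddle Team Dark cosmetic pickaxe swing, classed as player melee weapon audio.
# A quality-level root selects a six-variant random swing, a three-variant subset, or, for its last two levels, a shared mix.
# That mix layers a close DarkBomber-folder swing with a medium-attenuated distant fire-axe swing, both pitch-randomized; the random branches use close attenuation.
# A per-owner concurrency override caps simultaneous swings, stopping the oldest.
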SoundCue /BRCosmetics/Sounds/PickAxes/Cuddle_Team_Dark/PA_CuddleTeamDark_Swing_Cue
root=QualityLevel_0
nodes [
 {id=Attenuation_0, children [Modulator_3], properties {attenuationSettings=/Game/Sounds/Attenuation/Fort_Weapon_Pickaxe_Close_Att}}
 {id=Attenuation_1, children [Modulator_4], properties {attenuationSettings=/Game/Sounds/Attenuation/Fort_Weapon_Pickaxe_Medium_Att}}
 {id=Attenuation_2, children [Modulator_1], properties {attenuationSettings=&Fort_Weapon_Pickaxe_Close_Att}}
 {id=Attenuation_4, children [Random_5], properties {attenuationSettings=&Fort_Weapon_Pickaxe_Close_Att}}
 {id=Mixer_0, children [Attenuation_0, Attenuation_1], properties {inputVolume [1.2,1]}}
 {id=Modulator_1, children [Random_4]}
 {id=Modulator_2, children [Attenuation_4]}
 {id=Modulator_3, children [WavePlayer_15], properties {pitchMax=1.2, pitchMin=0.9}}
 {id=Modulator_4, children [WavePlayer_28], properties {pitchMax=1.2, pitchMin=0.9}}
 {id=QualityLevel_0, children [Attenuation_2, Modulator_2, Mixer_0, Mixer_0]}
 {
  id=Random_4
  children [WavePlayer_5, WavePlayer_17, WavePlayer_21, WavePlayer_22, WavePlayer_23, WavePlayer_24]
  properties {weights [1,1,1,1,1,1]}
 }
 {id=Random_5, children [WavePlayer_25, WavePlayer_26, WavePlayer_27], properties {weights [1,1,1]}}
 {id=WavePlayer_15, properties {soundWaveAssetPtr=/BRCosmetics/Sounds/PickAxes/DarkBomber/PA_DarkUnicorn_Swing_01}}
 {id=WavePlayer_17, properties {soundWaveAssetPtr=&/PickAxe_CuddleTeamDark_Swing_02}}
 {id=WavePlayer_21, properties {soundWaveAssetPtr=&/PickAxe_CuddleTeamDark_Swing_03}}
 {id=WavePlayer_22, properties {soundWaveAssetPtr=&/PickAxe_CuddleTeamDark_Swing_04}}
 {id=WavePlayer_23, properties {soundWaveAssetPtr=&/PickAxe_CuddleTeamDark_Swing_05}}
 {id=WavePlayer_24, properties {soundWaveAssetPtr=&/PickAxe_CuddleTeamDark_Swing_06}}
 {id=WavePlayer_25, properties {soundWaveAssetPtr=&/PickAxe_CuddleTeamDark_Swing_01}}
 {id=WavePlayer_26, properties {soundWaveAssetPtr=&PickAxe_CuddleTeamDark_Swing_02}}
 {id=WavePlayer_27, properties {soundWaveAssetPtr=&PickAxe_CuddleTeamDark_Swing_03}}
 {id=WavePlayer_28, properties {soundWaveAssetPtr=/Game/Athena/Sounds/Weapons/Fire_Axe_Swing_Distant_06}}
 {id=WavePlayer_5, properties {soundWaveAssetPtr=&PickAxe_CuddleTeamDark_Swing_01}}
]
properties {
 audioPropertiesSheet=/BRCosmetics/Sounds/Settings/PropertySheets/Pickaxe/APS_Cosmetic_Pickaxe_Swing
 bOverrideConcurrency=true
 bPrimeOnLoad=true
 maxDistance=3000
 soundClassObject=&Fort_WeaponMelee_PSM
 volumeMultiplier=0.35
 concurrencyOverrides {bLimitToOwner=true, maxCount=2, resolutionRule=StopOldest}
 concurrencySet [&Player_Melee_Weapon_Conc]
}
